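# Editor Utility Widget art tool that draws a chosen material into a chosen Texture Render Target 2D.
# On Construct, two SinglePropertyViews bind to the widget itself, exposing its RenderTarget and Material variables for editing.
# The Render button clears the target to opaque black, then calls DrawMaterialToRenderTarget with the material.
# A separate SetAsset function casts an input asset to MaterialInterface and stores it as Material.
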EditorUtilityWidgetBlueprint /Game/ArtTools/Tech_Art/RenderTargets/EUW_DrawMaterialToRenderTarget
graphs [
 EventGraph {nodes [
   "On Clicked (EditorUtilityButton_21)"() {is=title K2Node_ComponentBoundEvent}
   n1 = Get RenderTarget() -> (RenderTarget:Texture Render Target 2D Object Reference)
   > ClearRenderTarget2D!(
    execute
    WorldContextObject:Object Reference
    TextureRenderTarget:Texture Render Target 2D Object Reference=@n1
    ClearColor:Linear Color Structure=(R=0,G=0,B=0,A=1)
   ) -> (then)
   > DrawMaterialToRenderTarget!(
    execute
    WorldContextObject:Object Reference
    TextureRenderTarget:Texture Render Target 2D Object Reference=@n1
    Material:Material Interface Object Reference=Get Material() -> (Material:Material Interface Object Reference)
   ) -> (then)
   Event Construct()
   > SetObject(
    self=[
     Get SinglePropertyView_0() -> (SinglePropertyView_0:Single Property View Object Reference)
     Get SinglePropertyView() -> (SinglePropertyView:Single Property View Object Reference)
    ]
    NewObject=Self-Reference()
   )
   SetObject()
   Event PreConstruct()
   Event Tick()
  ]}
 SetAsset {nodes [
   n1 = SetAsset!() -> (then, Asset:Object Reference) {is=title K2Node_FunctionEntry}
   > n2 = Cast To MaterialInterface!(execute, Object=@n1) -> (then, CastFailed, AsMaterial Interface:Material Interface Object Reference) {is=title K2Node_DynamicCast}
   > Set Material(Material:Material Interface Object Reference=@n2) -> (Output_Get:Material Interface Object Reference)
  ]}
]
variables [
 Material {type=MaterialInterface, default=null}
 RenderTarget {type=TextureRenderTarget2D, default=null}
]
widgets {
 info {rootWidgetClass=&VerticalBox, widgetCount=5}
 widgets [
  VerticalBox_41 {widgetClassPath=&VerticalBox, widget=VerticalBox_41}
  SinglePropertyView_0 {
   widgetClassPath=&SinglePropertyView
   parent=VerticalBox_41
   widget=SinglePropertyView_0
   properties {propertyName=RenderTarget}
   slot {class=VerticalBoxSlot}
  }
  SinglePropertyView {
   widgetClassPath=&SinglePropertyView
   parent=VerticalBox_41
   widget=SinglePropertyView
   properties {propertyName=Material}
   slot {class=VerticalBoxSlot}
  }
  EditorUtilityButton_21 {widgetClassPath=&EditorUtilityButton, parent=VerticalBox_41, widget=EditorUtilityButton_21, slot {class=VerticalBoxSlot}}
  TextBlock_28 {widgetClassPath=&TextBlock, parent=EditorUtilityButton_21, widget=TextBlock_28, properties {text=Render}, slot {class=ButtonSlot}}
 ]
}
