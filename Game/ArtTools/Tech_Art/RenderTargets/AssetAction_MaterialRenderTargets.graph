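# Editor asset action utility that opens the Draw Material to Render Target tool for an asset selected in the editor.
# Its Draw Material to Render Target function checks that a selection exists, spawns EUW_DrawMaterialToRenderTarget as a tab and passes it the first selected asset through SetAsset.
# An unconnected Python node, which would open the ImpostorBaker widget the same way, sits unused in the graph.
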EditorUtilityBlueprint /Game/ArtTools/Tech_Art/RenderTargets/AssetAction_MaterialRenderTargets
graphs [
 Draw Material to Render Target {nodes [
   Draw Material to Render Target() {is=title K2Node_FunctionEntry}
   > n1 = GetSelectedAssets()
   > Branch(Condition=Is Valid Index!(TargetArray:Array of Object References=@n1, IndexToTest:int) -> (ReturnValue:bool))
   > n2 = SpawnAndRegisterTab(self=Get EditorUtilitySubsystem(), InBlueprint=&/EUW_DrawMaterialToRenderTarget)
   > n3 = Cast To EUW_DrawMaterialToRenderTarget!(execute, Object=@n2) -> (then, CastFailed, AsEUW Draw Material to Render Target:EUW Draw Material To Render Target Object Reference) {is=title K2Node_DynamicCast}
   > SetAsset!(
    execute
    self:EUW Draw Material To Render Target Object Reference=@n3
    Asset:Object Reference="Get (a copy)"(Array:Array of Object References=@n1) -> (Output:Object Reference)
   ) -> (then)
   Execute Python Script!(
    execute
    PythonScript="import unreal\r\n\r\nasset_path_string = \"/ImpostorBaker/BP/EUW_Generate_Impostor\"\r\nasset = unreal.load_asset(asset_path_string)\r\nss = unreal.get_editor_subsystem(unreal.EditorUtilitySubsystem)\r\ntab = ss.spawn_and_register_tab(asset)\r\ntab.call_method(\"SetAsset\", (asset_in,))"
    asset_in:Wildcard
   ) -> (then, ReturnValue) {inputs [asset_in]}
  ]}
 EventGraph {links [], nodes []}
]
variables []
